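# Base material for jungle-floor scatter ferns in Asteria, a masked two-sided foliage surface enabled for instanced static meshes and Nanite.
# MF_JungleFerns supplies the material attributes, which pass through MF_MediumPlantAnimation for plant animation.
# The shared MF_Curie_Landscape function wraps the result last, with its grass switch on.
# All logic lives in these chained functions, joined through material attributes.
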Material /Game/Environments/Asteria/Foliage/JungleFloor/M_Base_JungleSctterFern
graph {
 properties {
  bUseMaterialAttributes=true
  bUsedWithInstancedStaticMeshes=true
  bUsedWithNanite=true
  blendMode=BLEND_Masked
  pixelDepthOffsetMode=PDOM_Legacy
  shadingModel=MSM_TwoSidedFoliage
  twoSided=true
  usedShadingModels=Two Sided Foliage
 }
 nodes [MaterialFunctionCall!(
   In=MaterialFunctionCall!(In=MaterialFunctionCall() {materialFunction=&/MF_JungleFerns}.Result) {materialFunction=/Game/Environments/Asteria/Foliage/Plants/BaseMaterial/MF_MediumPlantAnimation}.Result
   UseHeightFalloff=StaticBool()
   OnFireEmissiveOverride
   HeightFalloffRadius
   HeightFalloffHardness
   VariationMasksMobile
   In_Grass=StaticBool() {value=true}
   Is_Stylised_WE
  ) {materialFunction=/Game/Packages/Fortress_SharedMaterials/Base_Material_Functions/MF_Curie_Landscape}]
}
